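# Animation Blueprint for the Realm backpack cosmetic, a hovering droid driven by Fortnite's floating pet anim instance.
# A state machine plays random idles and folds up during melee or skydive dives.
# Bone modifiers add hover rotations, shrink it when DBNO-carried, then blend pack position between carrier offset and world-space target.
# Events retune spring controls for DBNO, hunker, crouch and melee, and set timed flags for firing and parachute opening.
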AnimBlueprint /BRCosmetics/Accessories/FORT_Backpacks/Backpack_M_MED_Realm/Meshes/M_MED_Realm_Pack_AnimBP
target_skeleton=&/M_MED_Realm_Pack_Skeleton
graphs [
 AnimGraph {nodes [
   Local Space Ref Pose()
   Random Sequence Player!() -> (Pose:Pose Link Structure) {
    is=title AnimGraphNode_RandomPlayer
    entries [
     {
      chanceToPlay=0.66
      maxLoopCount=0
      maxPlayRate=1
      minLoopCount=0
      minPlayRate=1
      sequence=/BRCosmetics/Accessories/FORT_Backpacks/Backpack_M_MED_Realm/Animations/Realm_Male_Backpack_idle_anim
      blendIn {blendOption=Linear, blendTime=0.2, customCurve=null}
     }
     {
      chanceToPlay=0.33
      maxLoopCount=0
      maxPlayRate=1
      minLoopCount=0
      minPlayRate=0
      sequence=/BRCosmetics/Accessories/FORT_Backpacks/Backpack_M_MED_Realm/Animations/Realm_Male_Backpack_idle_anim_break
      blendIn {blendOption=Linear, blendTime=0.2, customCurve=null}
     }
    ]
   }
   n1 = "Transform (Modify) Bone - Bone: pack"(
    ComponentPose=Local To Component(LocalPose=Use cached pose 'Pre Spring Pose'())
    Translation=Get BackpackTargetVector() -> (BackpackTargetVector:Vector)
    Rotation="0,0,0"
   ) {translationMode=BMM_Replace, translationSpace=BCS_WorldSpace, alphaBoolBlend {blendOption=Linear}, boneToModify {boneName=pack}}
   n2 = "Transform (Modify) Bone - Bone: pack"(
    ComponentPose=Local To Component(LocalPose=Use cached pose 'Pre Spring Pose'())
    Translation=Get CarrierOffsetTotal() -> (CarrierOffsetTotal:Vector)
    Rotation="0,0,0"
   ) {translationMode=BMM_Additive, alphaBoolBlend {blendOption=Linear}, boneToModify {boneName=pack}}
   Output Pose(Result=Two Way Blend(
     A=Component To Local(ComponentPose=@n2)
     B=Component To Local(ComponentPose=@n1)
     Alpha=Get BackpackTargetVectorAlphaSmooth() -> (BackpackTargetVectorAlphaSmooth:float)
    ) {blendNode {alphaBoolBlend {blendOption=Linear}}}) {layerGroup=None}
   AnimGraphNode_StateMachine_0 = New State Machine() {is=title AnimGraphNode_StateMachine}
   n3 = "Transform (Modify) Bone - Bone: dyn_droid"(
    ComponentPose=Local To Component(LocalPose=@AnimGraphNode_StateMachine_0)
    Alpha=1
    Translation="0,0,0"
    Rotation=Get HoverRotationOffsetTotal() -> (HoverRotationOffsetTotal:Rotator)
    Scale="1,1,1"
   ) {rotationMode=BMM_Additive, alphaBoolBlend {blendOption=Linear}, boneToModify {boneName=dyn_droid}}
   n4 = Get CarrierOffsetRotationVelocity() -> (CarrierOffsetRotationVelocity:Rotator)
   n5 = Get CarrierOffsetRotationVelocityAlpha() -> (CarrierOffsetRotationVelocityAlpha:float)
   n6 = "Transform (Modify) Bone - Bone: root"!(
    ComponentPose="Transform (Modify) Bone - Bone: pack"(
     ComponentPose="Transform (Modify) Bone - Bone: dyn_droid"(ComponentPose=@n3, Alpha=@n5, Translation="0,0,0", Rotation=@n4, Scale="1,1,1") {rotationMode=BMM_Additive, alphaBoolBlend {blendOption=Linear}, boneToModify {boneName=dyn_droid}}
     Translation="0,0,0"
     Rotation=Get RotationYawDefinedHoverOffset() -> (RotationYawDefinedHoverOffset:Rotator)
    ) {rotationMode=BMM_Additive, alphaBoolBlend {blendOption=Linear}, boneToModify {boneName=pack}}
    bAlphaBoolEnabled:bool=Get bIsDBNOCarried()
    Translation="0,0,0"
    Rotation="0,0,0"
    Scale="0.01,0.01,0.01"
   ) -> (Pose) {
    alphaInputType=Bool
    scaleMode=BMM_Replace
    scaleSpace=BCS_WorldSpace
    alphaBoolBlend {blendOption=Linear}
    boneToModify {boneName=root}
    scale=<0.01 0.01 0.01>
   }
   Save cached pose 'Pre Spring Pose'(Pose=Component To Local(ComponentPose=@n6)) {cacheName=Pre Spring Pose}
  ]}
 New State Machine {
  entry=Idle
  path=AnimGraph.AnimGraphNode_StateMachine_0.New State Machine
  ids=id
  nodes [
   AnimStateNode_0 = Idle()
   AnimStateNode_1 = FoldUpStart() {is=title AnimStateNode}
   AnimStateNode_2 = FoldUpLoop() {is=title AnimStateNode}
   AnimStateNode_3 = UnFold() {is=title AnimStateNode}
   AnimStateTransitionNode_0 = AnimStateTransitionNode()
   AnimStateTransitionNode_1 = AnimStateTransitionNode() {bAutomaticRuleBasedOnSequencePlayerInState=true, crossfadeDuration=0.1}
   AnimStateTransitionNode_2 = AnimStateTransitionNode()
   AnimStateTransitionNode_3 = AnimStateTransitionNode() {bAutomaticRuleBasedOnSequencePlayerInState=true}
   AnimStateTransitionNode_4 = AnimStateTransitionNode()
  ]
  links [
   {
    dst=UnFold
    rule=AnimGraph.AnimGraphNode_StateMachine_0.New State Machine.AnimStateTransitionNode_2.Transition
    src=FoldUpLoop
    via=AnimStateTransitionNode_2
   }
   {
    dst=FoldUpLoop
    rule=AnimGraph.AnimGraphNode_StateMachine_0.New State Machine.AnimStateTransitionNode_1.Transition
    src=FoldUpStart
    via=AnimStateTransitionNode_1
   }
   {
    dst=FoldUpStart
    rule=AnimGraph.AnimGraphNode_StateMachine_0.New State Machine.AnimStateTransitionNode_0.Transition
    src=Idle
    via=AnimStateTransitionNode_0
   }
   {
    dst=FoldUpStart
    rule=AnimGraph.AnimGraphNode_StateMachine_0.New State Machine.AnimStateTransitionNode_4.Transition
    src=Idle
    via=AnimStateTransitionNode_4
   }
   {
    dst=Idle
    rule=AnimGraph.AnimGraphNode_StateMachine_0.New State Machine.AnimStateTransitionNode_3.Transition
    src=UnFold
    via=AnimStateTransitionNode_3
   }
  ]
 }
 Idle {
  path=AnimGraph.AnimGraphNode_StateMachine_0.New State Machine.AnimStateNode_0.Idle
  nodes [
   n1 = Random Sequence Player() {entries [
     {
      chanceToPlay=0.66
      maxLoopCount=0
      maxPlayRate=1
      minLoopCount=0
      minPlayRate=1
      sequence=&Realm_Male_Backpack_idle_anim
      blendIn {blendOption=Linear, blendTime=0.2, customCurve=null}
     }
     {
      chanceToPlay=0.33
      maxLoopCount=0
      maxPlayRate=1
      minLoopCount=0
      minPlayRate=0
      sequence=&Realm_Male_Backpack_idle_anim_break
      blendIn {blendOption=Linear, blendTime=0.2, customCurve=null}
     }
    ]}
   Output Animation Pose(Result=@n1)
  ]
 }
 FoldUpStart {
  path=AnimGraph.AnimGraphNode_StateMachine_0.New State Machine.AnimStateNode_1.FoldUpStart
  nodes [
   n1 = Sequence Player 'Realm_Male_Backpack_Fold_In_anim'!(bLoopAnimation:bool=False) -> (Pose) {
    bLoopAnimation=false
    playRate=2.5
    sequence=/BRCosmetics/Accessories/FORT_Backpacks/Backpack_M_MED_Realm/Animations/Realm_Male_Backpack_Fold_In_anim
   }
   Output Animation Pose(Result=@n1)
  ]
 }
 FoldUpLoop {
  path=AnimGraph.AnimGraphNode_StateMachine_0.New State Machine.AnimStateNode_2.FoldUpLoop
  nodes [Output Animation Pose(Result=Sequence Player 'Realm_Male_Backpack_Fold_pose'() {sequence=/BRCosmetics/Accessories/FORT_Backpacks/Backpack_M_MED_Realm/Animations/Realm_Male_Backpack_Fold_pose})]
 }
 UnFold {
  path=AnimGraph.AnimGraphNode_StateMachine_0.New State Machine.AnimStateNode_3.UnFold
  nodes [
   n1 = Sequence Player 'Realm_Male_Backpack_Fold_Out_anim'!(bLoopAnimation:bool=False) -> (Pose) {bLoopAnimation=false, sequence=/BRCosmetics/Accessories/FORT_Backpacks/Backpack_M_MED_Realm/Animations/Realm_Male_Backpack_Fold_Out_anim}
   Output Animation Pose(Result=@n1)
  ]
 }
 Transition {
  path=AnimGraph.AnimGraphNode_StateMachine_0.New State Machine.AnimStateTransitionNode_0.Transition
  nodes [Result(bCanEnterTransition=Get bIsPlayingMeleeAnim())]
 }
 Transition {path=AnimGraph.AnimGraphNode_StateMachine_0.New State Machine.AnimStateTransitionNode_1.Transition, nodes [Result()]}
 Transition {
  path=AnimGraph.AnimGraphNode_StateMachine_0.New State Machine.AnimStateTransitionNode_2.Transition
  nodes [
   n1 = Get bIsSkydiveDiveMode()
   n2 = Get bIsPlayingMeleeAnim()
   Result(bCanEnterTransition=AND Boolean(A=NOT Boolean(A=@n2), B=NOT Boolean(A=@n1)))
  ]
 }
 Transition {path=AnimGraph.AnimGraphNode_StateMachine_0.New State Machine.AnimStateTransitionNode_3.Transition, nodes [Result()]}
 Transition {
  path=AnimGraph.AnimGraphNode_StateMachine_0.New State Machine.AnimStateTransitionNode_4.Transition
  nodes [Result(bCanEnterTransition=Get bIsSkydiveDiveMode())]
 }
 EventGraph {nodes [
   AnimNotify_DBNO_Entered() {is=title K2Node_Event}
   > UpdateSpringControls!(
    execute
    self:Fort Floating Pet Anim Instance Object Reference
    Stiffness:float=0.5
    StiffnessInterpSpeed:float=0.5
    StiffnessInterpSpeedSmooth:float=-1
    CriticalDampingFactor:float=-1
   ) -> (then)
   n1 = AnimNotify_Hunker_Entered() {is=title K2Node_Event}
   UpdateSpringControls(execute=[@n1.then, @n3.then], Stiffness=50, StiffnessInterpSpeed=8, StiffnessInterpSpeedSmooth=-1, CriticalDampingFactor=0.75)
   > Delay(Duration=1)
   > UpdateSpringControls(Stiffness=-1, StiffnessInterpSpeed=1, StiffnessInterpSpeedSmooth=-1, CriticalDampingFactor=-1)
   n2 = AnimNotify_Hunker_Exited() {is=title K2Node_Event}
   UpdateSpringControls(execute=[@n2.then, @n4.then], Stiffness=10, StiffnessInterpSpeed=0, StiffnessInterpSpeedSmooth=0, CriticalDampingFactor=0.5)
   Event CrouchStarted() {is=title K2Node_Event}
   > UpdateSpringControls(Stiffness=1, StiffnessInterpSpeed=-1, StiffnessInterpSpeedSmooth=-1, CriticalDampingFactor=0.9)
   > RetriggerableDelay(Duration=1)
   > UpdateSpringControls(Stiffness=10, StiffnessInterpSpeed=5, StiffnessInterpSpeedSmooth=-1, CriticalDampingFactor=0.5)
   Event CrouchStopped() {is=title K2Node_Event}
   > UpdateSpringControls(Stiffness=1, StiffnessInterpSpeed=-1, StiffnessInterpSpeedSmooth=-1, CriticalDampingFactor=0.9)
   > RetriggerableDelay(Duration=1)
   > UpdateSpringControls(Stiffness=10, StiffnessInterpSpeed=5, StiffnessInterpSpeedSmooth=-1, CriticalDampingFactor=0.5)
   n3 = Event MeleeStarted() {is=title K2Node_Event}
   n4 = Event MeleeStopped() {is=title K2Node_Event}
   Event ParachuteJustOpened() {is=title K2Node_Event}
   > Set bIsSkydivingParachuteJustOpened(bIsSkydivingParachuteJustOpened:bool=true) -> (Output_Get:bool)
   > RetriggerableDelay(Duration=1.2)
   > Set bIsSkydivingParachuteJustOpened(bIsSkydivingParachuteJustOpened=false)
   Event RecentlyFired() {is=title K2Node_Event}
   > Set bForceBackpackTargetVectorAlphaOnFromRecentlyFired(bForceBackpackTargetVectorAlphaOnFromRecentlyFired:bool=true) -> (Output_Get:bool)
   > RetriggerableDelay(Duration=0.5)
   > Set bForceBackpackTargetVectorAlphaOnFromRecentlyFired(bForceBackpackTargetVectorAlphaOnFromRecentlyFired=false)
   TryGetPawnOwner()
   Event BlueprintUpdateAnimation()
  ]}
]
variables []
