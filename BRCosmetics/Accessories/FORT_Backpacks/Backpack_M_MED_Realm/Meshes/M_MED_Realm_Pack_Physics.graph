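# Physics asset for the M_MED_Realm back bling (backpack cosmetic) skeletal mesh.
# It holds a single capsule body on the dyn_droid bone, offset from the bone origin and rotated a quarter turn.
# There are no constraints, so the body is not jointed to any other.
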PhysicsAsset /BRCosmetics/Accessories/FORT_Backpacks/Backpack_M_MED_Realm/Meshes/M_MED_Realm_Pack_Physics
bodies {dyn_droid [{shapeType=Capsule, length=27.7185, radius=18.5398, center=<-9.5339 0 0>, rotation=<rot 90 0 0>}]}
constraints []
